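# Editor debug utility that grants a configurable Juno loadout to the first player controller in the game world.
# Run casts to the Juno controller (saved as a bad cast node), drops current items as pickups, then waits two ticks.
# It gives equippables one at a time, trying to equip each, then gives other items by count.
# A branch's else path raises a script error about a missing Juno controller.
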EditorUtilityBlueprint /EditorDebugMenu/EUOs/Juno/Loadouts/EUO_GrantLoadout
graphs [EventGraph {nodes [
   Event Run()
   > n1 = GetGameWorld!(execute, self:Unreal Editor Subsystem Object Reference=Get UnrealEditorSubsystem()) -> (then, ReturnValue:World Object Reference)
   n2 = Bad cast node!(Object=GetPlayerController!(WorldContextObject:Object Reference=@n1, PlayerIndex:int=0) -> (ReturnValue:Player Controller Object Reference)) -> (bSuccess:bool=~, AsJuno Player Controller:Object Reference) {is=title K2Node_DynamicCast}
   > Set JunoPC(JunoPC:Object Reference=@n2.AsJuno Player Controller) -> (Output_Get:Object Reference)
   > DropAllItems!(
    execute
    self:Fort Player Controller Object Reference=@n2.AsJuno Player Controller
    IgnoreItemDef:Item Definition Base Object Reference
    AdditionalIgnoreItemDef:Item Definition Base Object Reference
    bIgnoreBuildingMaterials:bool=false
    bSpawnPickups:bool=true
   ) -> (then)
   n3 = Get JunoPC() -> (JunoPC:Object Reference)
   > DelayUntilNextTick!(execute, WorldContextObject:Object Reference=@n3) -> (then)
   > DelayUntilNextTick(WorldContextObject=@n3)
   > n4 = Branch()
   > n5 = Sequence()
   n6 = Get!() -> (
    Loadout_Items_9_#1:Array of S Juno Loadout Entry Structures
    Loadout_Equippables_10_#2:Array of S Juno Loadout Entry Structures
   )
   > n7 = For Each Loop!(Exec, Array:Array of S Juno Loadout Entry Structures=@n6.Loadout_Equippables_10_#2) -> (
    LoopBody
    Array Element_Item_15_#3:Fort World Item Definition Object Reference
    Array Element_Count_14_#4:int
    Array Index
    Completed
   )
   > For Loop(FirstIndex=1, LastIndex=@n7.Array Element_Count_14_#4)
   n8 = Get JunoPC()
   > n9 = GiveItemToInventoryOwner!(
    execute
    InventoryOwner:Fort Inventory Owner Interface Interface=@n8
    ItemDefinition:Fort World Item Definition Object Reference=@n7
    ItemVariantGuid:Guid Structure
    NumberToGive:int=1
    bNotifyPlayer:bool=true
    ItemLevel:int=-1
    PickupInstigatorHandle:int=-1
    bUseItemPickupAnalyticEvent:bool=false
    WeaponAmmoOverride:int=-1
   ) -> (then, ReturnValue:Fort World Item Object Reference)
   > Try Equipping Gear!(
    execute
    self:Object Reference=GetComponentByClass!(self:Actor Object Reference=@n8, ComponentClass:Actor Component Class Reference) -> (ReturnValue:Actor Component Object Reference)
    ItemToEquip:Fort Item Object Reference=@n9
   ) -> (then, ReturnValue:bool)
   n10 = For Each Loop(Exec=@n5.then_1, Array=@n6)
   > GiveItemToInventoryOwner(
    InventoryOwner=Get JunoPC()
    ItemDefinition=@n10
    NumberToGive=@n10.Array Element_Count_14_#4
    bNotifyPlayer=true
    ItemLevel=-1
    PickupInstigatorHandle=-1
    bUseItemPickupAnalyticEvent=false
    WeaponAmmoOverride=-1
   )
   ScriptError!(execute=@n4.else, Message:String="You are running a Juno loadout but don't have a juno PlayerController!!!") -> (then)
  ]}]
variables [
 Fort Player Controller {type=FortPlayerController, default=null}
 JunoPC {type=Object, default=null}
 Loadout {type=S_JunoLoadout, default {equippables [{count=1, item=null}, {count=1, item=null}], items [{count=10, item=null}, {count=80, item=null}]}}
]
